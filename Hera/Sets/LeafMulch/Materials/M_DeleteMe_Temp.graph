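# M_DeleteMe_Temp is a placeholder Default Lit material in the LeafMulch set, named as a temporary asset for deletion.
# It holds a single vector parameter with a magenta default, and the graph shows no connection from it to any material output.
# Automatic usage flag setting in the editor is turned off.
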Material /Hera/Sets/LeafMulch/Materials/M_DeleteMe_Temp
graph {
 properties {bAutomaticallySetUsageInEditor=false, usedShadingModels=Default Lit}
 nodes [VectorParameter() {defaultValue=<rgba 1 0 0.7366 1>}]
}
